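; Blueprint for a citrus tree foliage prop in the Asteria environment that manages camera collision around the player.
; When a FortPlayerPawn begins overlapping CameraCollider_0, that collider switches its Camera channel response to Ignore.
; When the pawn ends the overlap, the response returns to Block.
; BeginPlay, Tick and actor overlap events are unused stubs, and the construction script only calls its parent.
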(event EventBeginPlay
  (|CallParentFunction) ; Questo nodo è disabilitato e non verrà chiamato.\nTrascina i pin per creare funzionalità.
)

(event Collision|EventActorBeginOverlap
  (params (OtherActor :type Actor%20Object%20Reference))
  (|CallParentFunction OtherActor) ; Questo nodo è disabilitato e non verrà chiamato.\nTrascina i pin per creare funzionalità.
)

(event EventTick
  (params (DeltaSeconds :type Float%20%28single-precision%29))
  (|CallParentFunction DeltaSeconds) ; Questo nodo è disabilitato e non verrà chiamato.\nTrascina i pin per creare funzionalità.
)

(event OnComponentBeginOverlap(CameraCollider_0)
  (params (OverlappedComponent :type Primitive%20Component%20Object%20Reference) (OtherActor :type Actor%20Object%20Reference) (OtherComp :type Primitive%20Component%20Object%20Reference) (OtherBodyIndex :type Integer) (bFromSweep :type Boolean) (SweepResult :type Hit%20Result%20Structure%20%28by%20ref%29))
  (Utilities|Casting|CastToFortPlayerPawn OtherActor)
  (Collision|SetCollisionResponsetoChannel (Variables|Default|GetCameraCollider_0) "ECC_Camera" "ECR_Ignore"))

(event OnComponentEndOverlap(CameraCollider_0)
  (params (OverlappedComponent :type Primitive%20Component%20Object%20Reference) (OtherActor :type Actor%20Object%20Reference) (OtherComp :type Primitive%20Component%20Object%20Reference) (OtherBodyIndex :type Integer))
  (Utilities|Casting|CastToFortPlayerPawn OtherActor)
  (Collision|SetCollisionResponsetoChannel (Variables|Default|GetCameraCollider_0) "ECC_Camera" "ECR_Block"))

(fn ConstructionScript
  (|CallParentFunction))
